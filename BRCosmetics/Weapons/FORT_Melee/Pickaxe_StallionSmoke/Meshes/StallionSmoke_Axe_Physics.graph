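# Physics asset for the StallionSmoke two-handed pickaxe mesh, giving its handle and hanging tassel collision bodies and joints.
# The handle uses two capsules, and each of the four dyn_tassel bones has its own capsule.
# Constraints chain the tassel bone to bone from the handle, with wide swing limits and tight twist limits so it can dangle and swing without spinning.
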PhysicsAsset /BRCosmetics/Weapons/FORT_Melee/Pickaxe_StallionSmoke/Meshes/StallionSmoke_Axe_Physics
bodies {
 Melee_TwoHanded_Handle [
  {shapeType=Capsule, length=55.0516, radius=1.9138, center=<0.2538 -1.1806 -7.1214>, rotation=<rot 0.0974 -90 -0.6137>}
  {
   shapeType=Capsule
   length=4.5591
   radius=2.9409
   shapeName=Melee_TwoHanded_Handle_capsule
   center=<0.6025 -1.2307 -37.9066>
   rotation=<rot 0 -90 0>
  }
 ]
 dyn_tassel_01 [{shapeType=Capsule, length=2.8, radius=1.3, shapeName=dyn_tassel_01_capsule, rotation=<rot 0 0 90>}]
 dyn_tassel_02 [{shapeType=Capsule, length=2.8, radius=1.3, shapeName=dyn_tassel_02_capsule, rotation=<rot 0 0 90>}]
 dyn_tassel_03 [{shapeType=Capsule, length=2.8, radius=2.2, shapeName=dyn_tassel_03_capsule, rotation=<rot 0 0 90>}]
 dyn_tassel_04 [{shapeType=Capsule, length=1.9, radius=2.26, shapeName=dyn_tassel_04_capsule, center=<0.36 0 0>}]
}
constraints [
 {bone1Name=dyn_tassel_01, bone2Name=Melee_TwoHanded_Handle, swing1LimitDegrees=65, swing2LimitDegrees=65, twistLimitDegrees=10}
 {bone1Name=dyn_tassel_02, bone2Name=dyn_tassel_01, swing1LimitDegrees=90, swing2LimitDegrees=90, twistLimitDegrees=10}
 {bone1Name=dyn_tassel_03, bone2Name=dyn_tassel_02, swing1LimitDegrees=90, swing2LimitDegrees=90, twistLimitDegrees=10}
 {bone1Name=dyn_tassel_04, bone2Name=dyn_tassel_03, swing1LimitDegrees=80, swing2LimitDegrees=80, twistLimitDegrees=10}
]
